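# Physics asset that provides cloth collision for the capture-the-flag flag gadget in Creative mode, going by its name and path.
# It holds a single capsule body on the Melee_TwoHanded_Handle bone, long and thin and offset along the bone like a pole.
# There are no constraints, so it defines only a collision shape rather than a jointed ragdoll.
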PhysicsAsset /Game/Creative/Devices/CTF_Flag/Meshes/CTF_Flag_Gadget_ClothCollision
bodies {Melee_TwoHanded_Handle [{shapeType=Capsule, length=127.3155, radius=3, center=<-0.0001 -0.9483 46.2948>}]}
constraints []
