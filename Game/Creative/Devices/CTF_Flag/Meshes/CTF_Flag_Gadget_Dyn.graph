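# Physics asset for the Capture the Flag flag gadget, giving its cloth bones physics bodies so the flag can swing from the pole.
# A long capsule on Melee_TwoHanded_Handle anchors three six-bone capsule chains for the top, middle and bottom of the flag.
# Each chain links bone to bone with swing limits generally tightest on top and loosest at the bottom, while twist stays narrow throughout.
# No constraints join the three rows to each other.
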PhysicsAsset /Game/Creative/Devices/CTF_Flag/Meshes/CTF_Flag_Gadget_Dyn
bodies {
 Melee_TwoHanded_Handle [{shapeType=Capsule, length=78.269, radius=5.0875, center=<-0.0057 -1.6994 34.9082>, rotation=<rot -0.166 -90 -0.0022>}]
 dyn_flag_bott_01 [{shapeType=Capsule, length=10, radius=2, center=<6 0 0>, rotation=<rot 90 0 0>}]
 dyn_flag_bott_02 [{shapeType=Capsule, length=10, radius=2, center=<7 0 0>, rotation=<rot 90 0 0>}]
 dyn_flag_bott_03 [{shapeType=Capsule, length=5, radius=2, center=<4 0 0>, rotation=<rot 90 0 0>}]
 dyn_flag_bott_04 [{shapeType=Capsule, length=5, radius=2, center=<4 0 0>, rotation=<rot 90 0 0>}]
 dyn_flag_bott_05 [{shapeType=Capsule, length=4, radius=2, center=<2.2 0 0>, rotation=<rot 90 0 0>}]
 dyn_flag_bott_06 [{shapeType=Capsule, length=3, radius=2, center=<1.5 0 0>, rotation=<rot 90 0 0>}]
 dyn_flag_mid_01 [{shapeType=Capsule, length=10, radius=2, center=<6 0 0>, rotation=<rot 90 0 0>}]
 dyn_flag_mid_02 [{shapeType=Capsule, length=10, radius=2, center=<7 0 0>, rotation=<rot 90 0 0>}]
 dyn_flag_mid_03 [{shapeType=Capsule, length=5, radius=2, center=<4 0 0>, rotation=<rot 90 0 0>}]
 dyn_flag_mid_04 [{shapeType=Capsule, length=5, radius=2, center=<4 0 0>, rotation=<rot 90 0 0>}]
 dyn_flag_mid_05 [{shapeType=Capsule, length=4, radius=2, center=<2.2 0 0>, rotation=<rot 90 0 0>}]
 dyn_flag_mid_06 [{shapeType=Capsule, length=3, radius=2, center=<1.5 0 0>, rotation=<rot 90 0 0>}]
 dyn_flag_top_01 [{shapeType=Capsule, length=10, radius=2, center=<6 0 0>, rotation=<rot 90 0 0>}]
 dyn_flag_top_02 [{shapeType=Capsule, length=10, radius=2, center=<7 0 0>, rotation=<rot 90 0 0>}]
 dyn_flag_top_03 [{shapeType=Capsule, length=5, radius=2, center=<4 0 0>, rotation=<rot 90 0 0>}]
 dyn_flag_top_04 [{shapeType=Capsule, length=5, radius=2, center=<4 0 0>, rotation=<rot 90 0 0>}]
 dyn_flag_top_05 [{shapeType=Capsule, length=4, radius=2, center=<2.2 0 0>, rotation=<rot 90 0 0>}]
 dyn_flag_top_06 [{shapeType=Capsule, length=3, radius=2, center=<1.5 0 0>, rotation=<rot 90 0 0>}]
}
constraints [
 {bone1Name=dyn_flag_top_01, bone2Name=Melee_TwoHanded_Handle, swing1LimitDegrees=20, swing2LimitDegrees=20, twistLimitDegrees=10}
 {bone1Name=dyn_flag_top_02, bone2Name=dyn_flag_top_01, swing1LimitDegrees=25, swing2LimitDegrees=25, twistLimitDegrees=10}
 {bone1Name=dyn_flag_top_03, bone2Name=dyn_flag_top_02, swing1LimitDegrees=40, swing2LimitDegrees=40, twistLimitDegrees=10}
 {bone1Name=dyn_flag_top_04, bone2Name=dyn_flag_top_03, swing1LimitDegrees=40, swing2LimitDegrees=40, twistLimitDegrees=10}
 {bone1Name=dyn_flag_top_05, bone2Name=dyn_flag_top_04, swing1LimitDegrees=40, swing2LimitDegrees=40, twistLimitDegrees=10}
 {bone1Name=dyn_flag_top_06, bone2Name=dyn_flag_top_05, swing1LimitDegrees=40, swing2LimitDegrees=40, twistLimitDegrees=10}
 {bone1Name=dyn_flag_bott_01, bone2Name=Melee_TwoHanded_Handle, swing1LimitDegrees=60, swing2LimitDegrees=60, twistLimitDegrees=10}
 {bone1Name=dyn_flag_bott_02, bone2Name=dyn_flag_bott_01, swing1LimitDegrees=60, swing2LimitDegrees=60, twistLimitDegrees=10}
 {bone1Name=dyn_flag_bott_03, bone2Name=dyn_flag_bott_02, swing1LimitDegrees=60, swing2LimitDegrees=60, twistLimitDegrees=10}
 {bone1Name=dyn_flag_bott_04, bone2Name=dyn_flag_bott_03, swing1LimitDegrees=60, swing2LimitDegrees=60, twistLimitDegrees=10}
 {bone1Name=dyn_flag_bott_05, bone2Name=dyn_flag_bott_04, swing1LimitDegrees=60, swing2LimitDegrees=60, twistLimitDegrees=10}
 {bone1Name=dyn_flag_bott_06, bone2Name=dyn_flag_bott_05, swing1LimitDegrees=60, swing2LimitDegrees=60, twistLimitDegrees=10}
 {bone1Name=dyn_flag_mid_01, bone2Name=Melee_TwoHanded_Handle, swing1LimitDegrees=45, swing2LimitDegrees=45, twistLimitDegrees=10}
 {bone1Name=dyn_flag_mid_02, bone2Name=dyn_flag_mid_01, swing1LimitDegrees=40, swing2LimitDegrees=40, twistLimitDegrees=10}
 {bone1Name=dyn_flag_mid_03, bone2Name=dyn_flag_mid_02, swing1LimitDegrees=40, swing2LimitDegrees=40, twistLimitDegrees=10}
 {bone1Name=dyn_flag_mid_04, bone2Name=dyn_flag_mid_03, swing1LimitDegrees=40, swing2LimitDegrees=40, twistLimitDegrees=10}
 {bone1Name=dyn_flag_mid_05, bone2Name=dyn_flag_mid_04, swing1LimitDegrees=40, swing2LimitDegrees=40, twistLimitDegrees=10}
 {bone1Name=dyn_flag_mid_06, bone2Name=dyn_flag_mid_05, swing1LimitDegrees=40, swing2LimitDegrees=40, twistLimitDegrees=10}
]
